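# Unlit translucent mesh-particle material for the Evil Cowboy pickaxe head, rendering tinted, smoky wisps.
# Three panning smoke and noise textures, offset by particle red, are summed and masked by inverted Fresnel and UV edges.
# This mask drives particle-green-scaled emissive and opacity with dynamic erosion, depth fade and a UV1 gradient.
# Dynamic-parameter vector noise in actor-relative space, blended by that gradient, supplies likely vertex displacement, replaced by a constant on mobile.
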Material /BRCosmetics/Athena/Effects/Materials/Pickaxe/M_Pickaxe_EvilCowboy_Pickaxehead
graph {
 properties {
  bUsedWithMeshParticles=true
  bUsedWithNiagaraMeshParticles=true
  blendMode=BLEND_Translucent
  pixelDepthOffsetMode=PDOM_Legacy
  previewMesh=/Game/Effects/Fort_Effects/Mesh/PickAxe/SK_Pickaxe_EvilCowboy_Blade_MeshP
  shadingModel=MSM_Unlit
  twoSided=true
  usedShadingModels=Unlit
 }
 nodes [
  n1 = VectorParameter() {parameterName=sadg, defaultValue=<rgba 0.2221 0.0154 0.25 1>}
  n2 = ParticleColor()
  n3 = Power(Base=Fresnel(ExponentIn=ScalarParameter() {defaultValue=1, parameterName=SmokeFrensnel})) {constExponent=4}
  n4 = ParticleColor()
  n5 = TextureCoordinate() {uTiling=3}
  n6 = Multiply(A=ScalarParameter() {defaultValue=1, parameterName=XTileMult_2}, B=ComponentMask(None=@n5) {r=true})
  n7 = Panner(Coordinate=AppendVector(A=Add(A=@n4.R, B=@n6), B=ComponentMask(None=@n5) {g=true})) {speedX=1}
  n8 = ParticleColor()
  n9 = TextureCoordinate() {uTiling=2, vTiling=0.65}
  n10 = Multiply(A=ComponentMask(None=@n9) {r=true}, B=ScalarParameter() {defaultValue=1, parameterName=XTileMult})
  n11 = ScalarParameter() {defaultValue=0.5, parameterName=PanSpeed}
  n12 = Panner(Coordinate=AppendVector(A=Add(A=@n8.R, B=@n10), B=ComponentMask(None=@n9) {g=true}), Speed=AppendVector(A=Constant(), B=@n11)) {speedY=0.5}
  n13 = ScalarParameter() {defaultValue=1, parameterName=XTileMult_1}
  n14 = TextureCoordinate() {uTiling=2, vTiling=0.65}
  n15 = ComponentMask(None=@n14) {r=true}
  n16 = AppendVector(A=Add(A=ParticleColor().R, B=Multiply(A=@n13, B=@n15)), B=ComponentMask(None=@n14) {g=true})
  n17 = AppendVector(A=Constant(), B=@n11)
  n18 = Add(
   A=Add(
    A=TextureSample(UVs=@n7) {samplerType=SAMPLERTYPE_Grayscale, texture=&T_Smoke_Trail_Soft}.G
    B=TextureSample(UVs=@n12) {texture=/Game/Weapons/FORT_Melee/Effects/LaserSwordFounders/T_LaserSwordTrailNoise}.G
   )
   B=TextureSample(UVs=Panner(Coordinate=@n16, Speed=@n17) {speedY=0.5}) {texture=&T_Inky_Smoke_Tile}.G
  )
  n19 = TextureCoordinate()
  n20 = ComponentMask(None=@n19) {r=true}
  n21 = OneMinus(None=@n20)
  n22 = ComponentMask(None=@n19) {g=true}
  n23 = OneMinus(None=@n22)
  n24 = Multiply(
   A=Multiply(A=OneMinus(None=@n3), B=Power(Base=@n18) {constExponent=1.25})
   B=Multiply(A=Multiply(A=@n20, B=@n21), B=Multiply(A=@n22, B=@n23))
  )
  MaterialFunctionCall(Emissive=Multiply(A=Multiply(A=@n1.RGB, B=@n2.G), B=@n24)) {materialFunction=&MF_FX_EmissiveAdjustments}
  n25 = MaterialFunctionCall!(
   Dynamic Param=Multiply(A=@n24, B=ParticleColor().A)
   Alpha=DynamicParameter() {defaultValue=<rgba 1 50 15 1>, paramNames [Curl_Speed, Curl_Scale, Curl_Intensity, Erode]}.Erode
  ) {materialFunction=/Game/Effects/Fort_Effects/Materials/_Functions/Erode_Alpha}
  n26 = ScalarParameter() {defaultValue=10, parameterName=FadeDistance}
  n27 = TextureCoordinate() {coordinateIndex=1, vTiling=1.35}
  n28 = Clamp(None=Power(Base=ComponentMask(None=@n27) {g=true}))
  Multiply(A=Multiply(A=DepthFade(Opacity=@n25.Result, FadeDistance=@n26), B=ScalarParameter() {defaultValue=0.25, parameterName=MultOpacity}), B=@n28)
  n29 = DynamicParameter() {defaultValue=<rgba 1 50 15 1>, paramNames [Curl_Speed, Curl_Scale, Curl_Intensity, Erode]}
  n30 = Multiply(A=@n29.Curl_Intensity) {constB=0.15}
  n31 = WorldPosition()
  n32 = ActorPositionWS()
  n33 = Divide(
   A=Subtract(
    A=Subtract(A=@n31.XYZ, B=@n32)
    B=DynamicParameter() {defaultValue=<rgba 1 50 15 1>, paramNames [Curl_Speed, Curl_Scale, Curl_Intensity, Erode]}.Curl_Speed
   )
   B=DynamicParameter() {defaultValue=<rgba 1 50 15 1>, paramNames [Curl_Speed, Curl_Scale, Curl_Intensity, Erode]}.Curl_Scale
  )
  ShadingPathSwitch(
   Default=Multiply(A=LinearInterpolate(A=@n29.Curl_Intensity, B=@n30, Alpha=@n28), B=VectorNoise(World Position=@n33) {noiseFunction=VNF_VectorALU})
   Mobile=Constant() {r=1}
  ) {desc=Migrated from FeatureLevelSwitch}
 ]
}
